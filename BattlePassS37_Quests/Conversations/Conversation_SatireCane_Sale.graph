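# Conversation database for the SatireCane NPC's sale dialogue in Battle Pass S37 quests.
# One entry point branches into Buy, Enhanced Patch Up and Rift-To-Go choices, each gated by its service tag and offering Back.
# Buy also requires a community bonus-goal completion token, then sells loot with a give-item animation and repeats.
# Patch Up heals the player using curve-table values and stays visible but fails when the health check is unmet.
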ConversationDatabase /BattlePassS37_Quests/Conversations/Conversation_SatireCane_Sale
entry_points [{tag=Conversation.Entry.SatireCane.Sale, destinations [#1]}]
nodes [
 ConversationEntryPointNode_0 {
  id=#1
  class=ConversationEntryPointNode
  outputs [#2, #3, #4]
  properties {entryTag {tagName=Conversation.Entry.SatireCane.Sale}}
 }
 Speech_C_2 {
  id=#4
  class=Speech_C
  outputs [#5, #6]
  properties {
   bUsePerNPC_ContextualMessageLookup=true
   softTaskIcon=/Game/Athena/HUD/Interaction/T_UI_NPC_Rift
   generalConfig {defaultMessage="It's riftin' time!"}
  }
  sub_nodes [
   FortConversationRequirement_HasService_7 {class=FortConversationRequirement_HasService, properties {serviceTag {tagName=Tandem.Service.RiftToGo}}}
   FortConversationChoiceWithPreviewNode_7 {
    class=FortConversationChoiceWithPreviewNode
    properties {defaultChoiceDisplayText=RIFT-TO-GO, choiceTags {gameplayTags [{tagName=Conversation.Choice.CustomSortPriority.Low}]}}
   }
  ]
 }
 RiftToGo_C_0 {
  id=#6
  class=RiftToGo_C
  outputs []
  properties {softTaskIcon=&T_UI_NPC_Rift}
  sub_nodes [ConversationChoiceNode_9 {class=ConversationChoiceNode, properties {defaultChoiceDisplayText=CONFIRM}}]
 }
 Back_C_4 {
  id=#5
  class=Back_C
  sub_nodes [ConversationChoiceNode_2 {class=ConversationChoiceNode, properties {defaultChoiceDisplayText=BACK}}]
 }
 Speech_C_1 {
  id=#3
  class=Speech_C
  outputs [#7, #8]
  properties {
   bUsePerNPC_ContextualMessageLookup=true
   softTaskIcon=/Game/Athena/HUD/Interaction/T_UI_NPC_medic
   generalConfig {defaultMessage=Anything for a friend.}
  }
  sub_nodes [
   FortConversationChoiceWithPreviewNode_13 {
    class=FortConversationChoiceWithPreviewNode
    properties {defaultChoiceDisplayText=ENHANCED PATCH UP, choiceTags {gameplayTags [{tagName=Conversation.Choice.CustomSortPriority.Low}]}}
   }
   FortConversationRequirement_HasService_18 {class=FortConversationRequirement_HasService, properties {serviceTag {tagName=Tandem.Service.EnhancedPatchUp}}}
  ]
 }
 EnhancedPatchUp_C_0 {
  id=#8
  class=EnhancedPatchUp_C
  outputs []
  properties {
   effectRecipientConfigs [{
     recipient=Player
     effects [{
       class=FortControllerEffect_HealPlayer
       properties {
        gameplayEffect=/NPCLibrary/Conversation/GameplayEffects/GE_CBGA_NPCService_PatchUp.GE_CBGA_NPCService_PatchUp_C
        healthRestoreQuantity {value=1, curve {curveTable=/Game/Athena/Balance/DataTables/NPCServiceSettings, rowName=EnhancedPatchUp.HealthRestoreQuantity}}
        parameterName {tagName=SetByCaller.Magnitude}
       }
      }]
    }]
   requirements [{
     cannotUseReasonParameter=Healthy
     failureNodeBehaviour=FailedButVisible
     participantId {tagName=Conversation.Participant.Player}
     requirement {
      class=FortControllerRequirement_AttributeThresholdCheck
      properties {
       queryAttribute {attribute=/Script/FortniteGame.FortHealthSet:Health, attributeName=Health, attributeOwner=&FortHealthSet}
       thresholdTestConfig {quantity {value=100, curve {rowName=Tandem.Gift.Health.Requirements.HealthThreshold}}}
      }
     }
    }]
  }
  sub_nodes [ConversationChoiceNode_25 {class=ConversationChoiceNode, properties {defaultChoiceDisplayText=ACCEPT}}]
 }
 Back_C_13 {
  id=#7
  class=Back_C
  sub_nodes [ConversationChoiceNode_40 {class=ConversationChoiceNode, properties {defaultChoiceDisplayText=BACK}}]
 }
 Speech_C_0 {
  id=#2
  class=Speech_C
  outputs [#9, #10]
  properties {nodeName="", softTaskIcon=&T_UI_NPC_question, generalConfig {defaultMessage=Sometimes you gotta blast 'em.}}
  sub_nodes [
   FortConversationRequirement_HasService_3 {class=FortConversationRequirement_HasService, properties {serviceTag {tagName=Tandem.Service.Sell1}}}
   FortConversationChoiceWithPreviewNode_2 {
    class=FortConversationChoiceWithPreviewNode
    properties {defaultChoiceDisplayText=BUY, choiceTags {gameplayTags [{tagName=Conversation.Choice.CustomSortPriority.Highest}]}}
   }
   FortConversationRequirement_ControllerRequirement_0 {
    class=FortConversationRequirement_ControllerRequirement
    properties {
     nodeName=UFortControllerRequirement_TokenThreshold_C_0
     requirement {
      class=UFortControllerRequirement_TokenThreshold_C
      properties {
       itemTemplateId=Token:token_br_s37_community_bonusgoal_complete_q05
       requiredQuantityThresholdConfig {requirement=GreaterThanOrEqual, quantity {value=1}}
      }
     }
    }
   }
  ]
 }
 Back_C_7 {
  id=#10
  class=Back_C
  sub_nodes [ConversationChoiceNode_21 {class=ConversationChoiceNode, properties {defaultChoiceDisplayText=BACK}}]
 }
 SellLoot_C_3 {
  id=#9
  class=SellLoot_C
  outputs [#11]
  sub_nodes [SideEffect_Client_PlayAnimationEvent_C_4 {class=SideEffect_Client_PlayAnimationEvent_C, properties {animationEventName=aie.npc.GiveItem}}]
 }
 Repeat_C_1 {id=#11, class=Repeat_C}
]
